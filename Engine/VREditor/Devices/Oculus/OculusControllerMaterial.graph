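# Surface material for the Oculus controller mesh in VR Editor, with parameter-driven button highlights, glowing trim and a power light.
# Vertex color channels, squared or multiplied by inverse alpha, form masks that scalar parameters scale for trigger, grip, directional and two buttons, tinted by InteractColor.
# These sum with switchable trim glow and a power light masked by a reused Vive texture.
# Vertex colors also blend gray shade and roughness-like values; a separate branch builds a world-space sine grid.
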Material /Engine/VREditor/Devices/Oculus/OculusControllerMaterial
graph {
 properties {pixelDepthOffsetMode=PDOM_Legacy, previewMesh=&/OculusControllerMesh, usedShadingModels=Default Lit}
 nodes [
  n1 = VectorParameter() {parameterName=TrimGlowColor, defaultValue=<rgba 0 0.4685 1 3>}
  n2 = Multiply(A=@n1.RGB, B=@n1.RGB)
  n3 = VertexColor()
  n4 = OneMinus(None=@n3.A)
  n5 = Multiply(A=@n3.R, B=@n4)
  n6 = VectorParameter() {group=Colors, parameterName=InteractColor, defaultValue=<rgba 1 1 1 0>}
  n7 = Multiply(A=ScalarParameter() {desc=Trigger, group=Buttons, parameterName=B2}, B=Multiply(A=@n3.B, B=@n3.B))
  n8 = Multiply(A=ScalarParameter() {desc=Grip, group=Buttons, parameterName=B1}, B=Multiply(A=@n3.B, B=@n4))
  n9 = ScalarParameter() {desc=Directional, group=Buttons, parameterName=B3}
  n10 = Multiply(A=@n3.R, B=@n3.R)
  n11 = ScalarParameter() {desc=Button2, group=Buttons, parameterName=B5}
  n12 = Multiply(A=@n3.G, B=@n4)
  n13 = ScalarParameter() {desc=Button1, group=Buttons, parameterName=B4}
  n14 = Multiply(A=@n3.G, B=@n3.G)
  n15 = Add(A=Add(A=Add(A=@n7, B=@n8), B=Multiply(A=@n9, B=@n10)), B=Add(A=Multiply(A=@n11, B=@n12), B=Multiply(A=@n13, B=@n14)))
  n16 = Multiply(
   A=TextureSample(UVs=TextureCoordinate() {uTiling=2, vTiling=2}) {texture=/Engine/VREditor/Devices/Vive/VivePreController_MasksB}.B
   B=ScalarParameter() {defaultValue=100, group=Power Light, parameterName=Power}
  )
  n17 = VectorParameter() {group=Power Light, parameterName=PowerColor, defaultValue=<rgba 0.5246 0.8842 1 0>}
  Add(
   A=StaticSwitchParameter(True=Multiply(A=@n2, B=@n5), False=Constant()) {defaultValue=true, parameterName=UseGlowingTrim}
   B=Add(A=Multiply(A=@n6.RGB, B=@n15), B=Multiply(A=@n16, B=@n17.RGB))
  )
  Add(A=@n12) {constB=0}
  Constant() {r=0.8}
  n18 = Add(A=Add(A=@n10, B=@n3.G), B=@n3.B)
  LinearInterpolate(A=Constant3Vector() {constant=<rgba 0.445 0.445 0.445 1>}, B=Constant3Vector() {constant=<rgba 0.055 0.055 0.055 1>}, Alpha=@n18)
  LinearInterpolate(Alpha=@n18) {constA=0.45, constB=0.7}
  n19 = Divide(A=WorldPosition().XYZ, B=ScalarParameter() {defaultValue=2, group=Grid, parameterName=GridSpacing})
  n20 = MaterialFunctionCall(Float3=Clamp(None=Sine(None=@n19) {period=2})) {materialFunction=&BreakOutFloat3Components}
  n21 = Max(A=Max(A=@n20.R, B=@n20.R), B=@n20.B)
  n22 = Constant() {r=1}
  n23 = ScalarParameter() {defaultValue=0.1, group=Grid, parameterName=Grid Width}
  Multiply(
   A=@n6.RGB
   B=Add(A=SphereMask(A=@n21, B=@n22, Radius=@n23) {attenuationRadius=0.1, hardnessPercent=0}, B=ScalarParameter() {group=Grid, parameterName=Bias}) {constB=0.05}
  )
 ]
}
